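# Physics asset of collision shapes for the F_LRG_ShimmerZen player body, apparently for colliding its cape, judging by the asset name.
# Capsules cover the head, clavicles, spine, pelvis and feet, plus dynamic jacket and boot bones, with paired capsules on spine_02 through spine_04.
# Arm, thigh, calf and neck_02 bodies carry no shapes.
# It defines no constraints, so it holds colliders only, with no ragdoll joints.
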PhysicsAsset /BRCosmetics/Characters/Player/Female/Large/Bodies/F_LRG_ShimmerZen/Meshes/F_LRG_ShimmerZen_CapeColliders
bodies {
 calf_l []
 calf_r []
 clavicle_l [{shapeType=Capsule, length=9, radius=9, center=<4 -1 -1>, rotation=<rot -87.1357 149.5189 35.4584>}]
 clavicle_r [{shapeType=Capsule, length=9, radius=9, center=<-4 1 1>, rotation=<rot -87.1356 149.5187 -144.5414>}]
 dyn_jacket_back_l [{shapeType=Capsule, length=10, radius=6, shapeName=dyn_jacket_back_l_capsule, center=<4 3.4686 0>}]
 dyn_jacket_front_l [{shapeType=Capsule, length=10, radius=6, shapeName=dyn_jacket_front_l_capsule, center=<4 3.4686 0>}]
 dyn_jacket_front_r [{shapeType=Capsule, length=10, radius=6, shapeName=dyn_jacket_front_r_capsule, center=<-4 -3.4686 0>}]
 dyn_jacket_side_l [{shapeType=Capsule, length=10, radius=6, shapeName=dyn_jacket_side_l_capsule, center=<4 3.4686 0>}]
 dyn_jacket_side_r [{shapeType=Capsule, length=10, radius=6, shapeName=dyn_jacket_side_r_capsule, center=<-4 -3.4686 0>}]
 dyn_l_boot_main [{shapeType=Capsule, length=12, radius=7, shapeName=dyn_l_boot_main_capsule, center=<-3 -6.2549 -1>, rotation=<rot 0 -90 0>}]
 dyn_r_boot_main [{shapeType=Capsule, length=12, radius=7, shapeName=dyn_r_boot_main_capsule, center=<3 7 1>, rotation=<rot 0 -90 0>}]
 foot_l [{shapeType=Capsule, length=11.0765, radius=8.0344, center=<-4.1811 6.0027 -0.338>, rotation=<rot 0.0388 1.4701 -89.3982>}]
 foot_r [{shapeType=Capsule, length=11.0765, radius=8.0344, center=<4.1811 -6.0027 0.338>, rotation=<rot 0.0388 1.4701 90.6018>}]
 head [{shapeType=Capsule, length=0, radius=18, center=<9.0697 0.0746 0>, rotation=<rot 0 -90 0>}]
 lowerarm_l []
 lowerarm_r []
 neck_02 []
 pelvis [{shapeType=Capsule, length=11, radius=14, center=<-1.6304 -3.863 -0.1981>}]
 spine_01 [{shapeType=Capsule, length=6, radius=16, center=<-6.1915 -1.8247 0>}]
 spine_02 [
  {shapeType=Capsule, length=2, radius=12, center=<-2.6885 -1.2641 0>}
  {shapeType=Capsule, length=16, radius=12, center=<-1.8317 0.1448 0>, rotation=<rot 90 0 0>}
 ]
 spine_03 [
  {shapeType=Capsule, length=7, radius=9, center=<0.0574 0.4416 0>}
  {shapeType=Capsule, length=15, radius=11.5, center=<-1.2163 0.549 0>, rotation=<rot 90 180 -175.0732>}
 ]
 spine_04 [
  {shapeType=Capsule, length=9.6, radius=12.5, center=<5.6234 -0.6779 0>}
  {shapeType=Capsule, length=10.0312, radius=12.5, center=<-1.2326 1.308 0>, rotation=<rot 90 0 0>}
 ]
 spine_05 [{shapeType=Capsule, length=9.6, radius=12.5, center=<0.6969 -0.1178 0>}]
 thigh_l []
 thigh_r []
 upperarm_l []
 upperarm_r []
}
constraints []
