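# Control Rig for the ShimmerZen large female face accessory that procedurally poses its long ponytail bone chain to follow the body.
# Construction caches the ponytail controls in an array and captures their pose.
# Forward solve parent-constrains each control's null to head or spine, shifting from head toward lower spine along the chain, then offsets some nulls.
# A non-stretch Hermite spline IK function fits the ponytail bones through those controls.
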ControlRigBlueprint /BRCosmetics/Characters/Player/Female/Large/Bodies/F_LRG_ShimmerZen/Meshes/Parts/F_LRG_ShimmerZen_FaceAcc_CtrlRig
graphs [
 ContainedGraph {nodes [
   n1 = RigVMFunctionEntryNode()
   SequenceAggregate = &RigVMFunction_Sequence(ExecuteContext=@n1.ExecuteContext)
   SequenceAggregate_1 = &RigVMFunction_Sequence(ExecuteContext=@SequenceAggregate.B)
   SequenceAggregate_2 = &RigVMFunction_Sequence(ExecuteContext=@SequenceAggregate_1.B)
   SequenceAggregate_3 = &RigVMFunction_Sequence(ExecuteContext=@SequenceAggregate_2.B)
   RigVMFunctionReturnNode(A=@SequenceAggregate.A, B=@SequenceAggregate_1.A, C=@SequenceAggregate_2.A, D=@SequenceAggregate_3.A, E=@SequenceAggregate_3.B)
  ]}
 ContainedGraph {nodes [
   n1 = RigVMFunctionEntryNode()
   SequenceAggregate = &RigVMFunction_Sequence(ExecuteContext=@n1.ExecuteContext)
   SequenceAggregate_1 = &RigVMFunction_Sequence(ExecuteContext=@SequenceAggregate.B)
   SequenceAggregate_2 = &RigVMFunction_Sequence(ExecuteContext=@SequenceAggregate_1.B)
   RigVMFunctionReturnNode(A=@SequenceAggregate.A, B=@SequenceAggregate_1.A, C=@SequenceAggregate_2.A, D=@SequenceAggregate_2.B)
  ]}
 RigVMFunctionLibrary {links [], nodes []}
 RigVMModel {nodes [
   RigUnit_BeginExecution = &RigUnit_BeginExecution()
   SequenceAggregate_1 = RigVMAggregateNode(ExecuteContext=@RigUnit_BeginExecution.ExecutePin)
   &RigUnit_ParentConstraint(
    AdvancedSettings=(RotationOrderForFilter=XZY)
    Child=(Type=Null,Name="dyn_ponytail_01_null")
    Filter=(TranslationFilter=(bX=true,bY=true,bZ=true),RotationFilter=(bX=true,bY=true,bZ=true),ScaleFilter=(bX=True,bY=True,bZ=True))
    Parents=((Item=(Type=Bone,Name="head"),Weight=1),(Item=(Type=Bone,Name="spine_05"),Weight=0))
    Weight=1
    ExecutePin=@SequenceAggregate_1.A
   )
   > &RigUnit_ParentConstraint(
    AdvancedSettings=(RotationOrderForFilter=XZY)
    Child=(Type=Null,Name="dyn_ponytail_02_null")
    Filter=(TranslationFilter=(bX=true,bY=true,bZ=true),RotationFilter=(bX=true,bY=true,bZ=true),ScaleFilter=(bX=True,bY=True,bZ=True))
    Parents=((Item=(Type=Bone,Name="head"),Weight=1),(Item=(Type=Bone,Name="spine_05"),Weight=0))
    Weight=1
   )
   > &RigUnit_ParentConstraint(
    AdvancedSettings=(RotationOrderForFilter=XZY)
    Child=(Type=Null,Name="dyn_ponytail_03_null")
    Filter=(TranslationFilter=(bX=True,bY=True,bZ=True),RotationFilter=(bX=true,bY=true,bZ=true),ScaleFilter=(bX=True,bY=True,bZ=True))
    Parents=((Item=(Type=Bone,Name="head"),Weight=1),(Item=(Type=Bone,Name="spine_05"),Weight=0))
    Weight=1
   )
   > &RigUnit_ParentConstraint(
    AdvancedSettings=(RotationOrderForFilter=XZY)
    Child=(Type=Null,Name="dyn_ponytail_04_null")
    Filter=(TranslationFilter=(bX=True,bY=True,bZ=True),RotationFilter=(bX=true,bY=true,bZ=true),ScaleFilter=(bX=True,bY=True,bZ=True))
    Parents=((Item=(Type=Bone,Name="head"),Weight=1),(Item=(Type=Bone,Name="spine_04"),Weight=0),(Item=(Type=Bone,Name="clavicle_l"),Weight=0))
    Weight=1
   )
   > &RigUnit_ParentConstraint(
    AdvancedSettings=(RotationOrderForFilter=XZY)
    Child=(Type=Null,Name="dyn_ponytail_05_null")
    Filter=(TranslationFilter=(bX=True,bY=True,bZ=True),RotationFilter=(bX=true,bY=true,bZ=true),ScaleFilter=(bX=True,bY=True,bZ=True))
    Parents=((Item=(Type=Bone,Name="head"),Weight=0.2),(Item=(Type=Bone,Name="spine_05"),Weight=0.8),(Item=(Type=Bone,Name="clavicle_r"),Weight=0))
    Weight=1
   )
   > &RigUnit_ParentConstraint(
    AdvancedSettings=(RotationOrderForFilter=XZY)
    Child=(Type=Null,Name="dyn_ponytail_06_null")
    Filter=(TranslationFilter=(bX=True,bY=True,bZ=True),RotationFilter=(bX=true,bY=true,bZ=true),ScaleFilter=(bX=True,bY=True,bZ=True))
    Parents=((Item=(Type=Bone,Name="None"),Weight=0),(Item=(Type=Bone,Name="spine_04"),Weight=1))
    Weight=1
   )
   > &RigUnit_ParentConstraint(
    AdvancedSettings=(RotationOrderForFilter=XZY)
    Child=(Type=Null,Name="dyn_ponytail_07_null")
    Filter=(TranslationFilter=(bX=True,bY=True,bZ=True),RotationFilter=(bX=true,bY=true,bZ=true),ScaleFilter=(bX=True,bY=True,bZ=True))
    Parents=((Item=(Type=Bone,Name="None"),Weight=0),(Item=(Type=Bone,Name="spine_03"),Weight=1))
    Weight=1
   )
   > &RigUnit_ParentConstraint(
    AdvancedSettings=(RotationOrderForFilter=XZY)
    Child=(Type=Null,Name="dyn_ponytail_08_null")
    Filter=(TranslationFilter=(bX=True,bY=True,bZ=True),RotationFilter=(bX=true,bY=true,bZ=true),ScaleFilter=(bX=True,bY=True,bZ=True))
    Parents=((Item=(Type=Bone,Name="None"),Weight=0),(Item=(Type=Bone,Name="spine_03"),Weight=1))
    Weight=1
   )
   > &RigUnit_ModifyTransforms(
    ItemToModify="((Item=(Type=Null,Name=\"dyn_ponytail_04_null\"),Transform=(Rotation=(X=0,Y=0,Z=0,W=1),Translation=(X=0,Y=0,Z=10),Scale3D=(X=1,Y=1,Z=1))),(Item=(Type=Null,Name=\"dyn_ponytail_06_null\"),Transform=(Rotation=(X=0,Y=0,Z=0,W=1),Translation=(X=0,Y=0,Z=10),Scale3D=( ...(truncated)"
    Weight=1
   )
   RigUnit_ItemArray = &RigUnit_ItemArray(Items="((Type=Bone,Name=\"dyn_ponytail_01\"),(Type=Bone,Name=\"dyn_ponytail_02\"),(Type=Bone,Name=\"dyn_ponytail_03\"),(Type=Bone,Name=\"dyn_ponytail_04\"),(Type=Bone,Name=\"dyn_ponytail_05\"),(Type=Bone,Name=\"dyn_ponytail_06\"),(Type=Bone,Name=\"dyn_ponytail_07\"),(Type=Bone ...(truncated)")
   > SplineIK_2_1 = RigVMFunctionReferenceNode(
    Primary Axis=(X=1,Y=0,Z=0)
    Samples Per Segment=1
    Secondary Spline Direction=(X=0,Y=0,Z=0)
    Spline Mode=Hermite
    Stretch=false
    Up Axis=(X=0,Y=0,Z=0)
    debug=true
    Bones=@RigUnit_ItemArray.Items
    Controls=RigVMVariableNode() {variable=dyn_Hair_Back_ctrls}.Value
   )
  ]}
 RigVMModel Construction Event Graph {nodes [
   n1 = &RigUnit_PrepareForExecution()
   SequenceAggregate = RigVMAggregateNode(ExecuteContext=@n1.ExecutePin)
   RerouteNode_6_1 = RigVMDispatchNode(Values="(((),Type=Control,Name=\"dyn_ponytail_01_ctrl\"),((),Type=Control,Name=\"dyn_ponytail_02_ctrl\"),((),Type=Control,Name=\"dyn_ponytail_03_ctrl\"),((),Type=Control,Name=\"dyn_ponytail_04_ctrl\"),((),Type=Control,Name=\"dyn_ponytail_05_ctrl\"),((),Type=Control,Name=\"dy ...(truncated)") {template="DISPATCH_RigVMDispatch_ArrayMake(in Values,out Array)"}
   RigVMVariableNode(Value=@RerouteNode_6_1.Array, ExecuteContext=@SequenceAggregate.A) {variable=dyn_Hair_Back_ctrls}
   > RigVMVariableNode(Value=/Script/ControlRig.RigUnit_HierarchyGetPoseItemArray(ElementType=All, Initial=false, ItemsToGet=@RerouteNode_6_1.Array) {is=struct RigVMUnitNode}.Pose) {variable=dyn_Hair_Back_pose}
  ]}
]
hierarchy [
 root {loc=<0 0 0>, rot=<0 0 0 1>}
 pelvis {parent=root, loc=<0 -0.648 94.7699>, rot=<0.0127 0.707 -0.0127 -0.707>}
 spine_01 {parent=pelvis, loc=<5.0975 -0.6388 0>, rot=<0 0 0.034 -0.9994>}
 spine_02 {parent=spine_01, loc=<7.252 -0.035 0>, rot=<0 0 0.0238 -0.9997>}
 spine_03 {parent=spine_02, loc=<6.0106 0 0>, rot=<0 0 0.0149 -0.9999>}
 spine_04 {parent=spine_03, loc=<8.1023 0 0>, rot=<0 0 -0.0132 -0.9999>}
 spine_05 {parent=spine_04, loc=<10.0142 0 0>, rot=<0 0 -0.0015 -1>}
 neck_01 {parent=spine_05, loc=<11.3775 -0.6396 0>, rot=<0 0 -0.2003 -0.9797>}
 neck_02 {parent=neck_01, loc=<4.5627 0 0>, rot=<0 0 0.0504 -0.9987>}
 head {parent=neck_02, loc=<5.3914 0 0>, rot=<0 0 0.115 -0.9934>}
 joint13 {parent=head, loc=<12.4126 -5.6645 0.077>, rot=<-0.6417 0.2969 0.2969 -0.6417>}
 joint21 {parent=joint13, loc=<2.4896 -0.1368 -0.4556>, rot=<0 0 0 1>}
 joint14 {parent=joint21, loc=<5.7488 0.1368 0.7678>, rot=<0 0.418 0 -0.9085>}
 joint15 {parent=joint14, loc=<8.2608 0 0.2915>, rot=<0 0.3839 0 -0.9234>}
 joint16 {parent=joint15, loc=<8.3455 0 0.1693>, rot=<0 0.2107 0 -0.9776>}
 joint17 {parent=joint16, loc=<10.7103 0 0.5289>, rot=<0 0.0952 0 -0.9955>}
 joint18 {parent=joint17, loc=<10.7609 0 0>, rot=<0 0 0 1>}
 joint19 {parent=joint18, loc=<10.7609 0 0>, rot=<0 0 0 1>}
 joint20 {parent=joint19, loc=<10.7609 0 0>, rot=<0 0 0 1>}
 joint11 {parent=joint19, loc=<5.1731 4.7547 1.4685>, rot=<-0.6681 0.2321 0.2173 -0.6727>}
 joint12 {parent=joint11, loc=<4.703 0 0>, rot=<0 0 0 1>}
 joint9 {parent=joint19, loc=<1.6995 -8.6154 -6.158>, rot=<-0.6472 -0.2846 0.1018 -0.6998>}
 joint10 {parent=joint9, loc=<6.4694 0 0>, rot=<0 0 0 1>}
 joint26 {parent=head, loc=<5.1327 5.3577 6.8204>, rot=<-0.7071 -0.0029 -0.0029 -0.7071>}
 joint27 {parent=joint26, loc=<-3.684 0 0>, rot=<0 0 0 1>}
 joint28 {parent=joint27, loc=<-3.685 0 0>, rot=<0 0 0 1>}
 joint29 {parent=joint28, loc=<-3.685 0 0>, rot=<0 0 0 1>}
 joint22 {parent=head, loc=<5.1331 5.3577 -6.8203>, rot=<-0.0029 0.7071 0.7071 -0.0029>}
 joint23 {parent=joint22, loc=<3.6848 0 0>, rot=<0 0 0 1>}
 joint24 {parent=joint23, loc=<3.6848 0 0>, rot=<0 0 0 1>}
 joint25 {parent=joint24, loc=<3.6848 0 0>, rot=<0 0 0 1>}
 dyn_hair_r_01 {parent=head, loc=<5.1327 5.3577 6.8204>, rot=<-0.7071 -0.0029 -0.0029 -0.7071>}
 dyn_hair_r_02 {parent=dyn_hair_r_01, loc=<-3.684 0 0>, rot=<0 0 0 1>}
 dyn_hair_r_03 {parent=dyn_hair_r_02, loc=<-3.685 0 0>, rot=<0 0 0 1>}
 dyn_hair_r_04 {parent=dyn_hair_r_03, loc=<-3.685 0 0>, rot=<0 0 0 1>}
 dyn_hair_l_01 {parent=head, loc=<5.1331 5.3577 -6.8203>, rot=<-0.0029 0.7071 0.7071 -0.0029>}
 dyn_hair_l_02 {parent=dyn_hair_l_01, loc=<3.6848 0 0>, rot=<0 0 0 1>}
 dyn_hair_l_03 {parent=dyn_hair_l_02, loc=<3.6848 0 0>, rot=<0 0 0 1>}
 dyn_hair_l_04 {parent=dyn_hair_l_03, loc=<3.6848 0 0>, rot=<0 0 0 1>}
 dyn_ponytail_Base {parent=head, loc=<12.4126 -5.6645 0.077>, rot=<-0.6417 0.2969 0.2969 -0.6417>}
 dyn_ponytail_01 {parent=dyn_ponytail_Base, loc=<2.4896 -0.1368 -0.4556>, rot=<0 0 0 1>}
 dyn_ponytail_02 {parent=dyn_ponytail_01, loc=<5.7488 0.1368 0.7678>, rot=<0 0.418 0 -0.9085>}
 dyn_ponytail_03 {parent=dyn_ponytail_02, loc=<8.2608 0 0.2915>, rot=<0 0.3839 0 -0.9234>}
 dyn_ponytail_04 {parent=dyn_ponytail_03, loc=<8.3455 0 0.1693>, rot=<0 0.2107 0 -0.9776>}
 dyn_ponytail_05 {parent=dyn_ponytail_04, loc=<10.7103 0 0.5289>, rot=<0 0.0952 0 -0.9955>}
 dyn_ponytail_06 {parent=dyn_ponytail_05, loc=<10.7609 0 0>, rot=<0 0 0 1>}
 dyn_ponytail_07 {parent=dyn_ponytail_06, loc=<10.7609 0 0>, rot=<0 0 0 1>}
 dyn_ponytail_08 {parent=dyn_ponytail_07, loc=<10.7609 0 0>, rot=<0 0 0 1>}
 dyn_ponytail_r_01 {parent=dyn_ponytail_07, loc=<5.1731 4.7547 1.4685>, rot=<-0.6681 0.2321 0.2173 -0.6727>}
 dyn_ponytail_r_02 {parent=dyn_ponytail_r_01, loc=<4.703 0 0>, rot=<0 0 0 1>}
 dyn_ponytail_l_01 {parent=dyn_ponytail_07, loc=<1.6995 -8.6154 -6.158>, rot=<-0.6472 -0.2846 0.1018 -0.6998>}
 dyn_ponytail_l_02 {parent=dyn_ponytail_l_01, loc=<6.4694 0 0>, rot=<0 0 0 1>}
 clavicle_l {parent=spine_05, loc=<4.1651 -0.3023 -5.0942>, rot=<0.0465 -0.7289 -0.009 -0.6829>}
 upperarm_l {parent=clavicle_l, loc=<11.2117 -1.1486 -0.0332>, rot=<0.0179 -0.2982 -0.028 -0.9539>}
 lowerarm_l {parent=upperarm_l, loc=<26.5013 -0.0835 0.2605>, rot=<0.0036 0.0161 -0.2298 -0.9731>}
 hand_l {parent=lowerarm_l, loc=<23.1356 -1.1606 -0.2612>, rot=<-0.6704 -0.0028 0.0567 -0.7398>}
 hand_attach_l {parent=hand_l, loc=<8.5 -3 0>, rot=<0.7071 0.7071 0 0>}
 weapon_l {parent=hand_l, loc=<-0.1974 -3.4111 0.3811>, rot=<0 0 0 1>}
 thumb_01_l {parent=hand_l, loc=<1.3118 -0.3191 -2.459>, rot=<0.6546 -0.4042 0.0664 -0.6354>}
 thumb_02_l {parent=thumb_01_l, loc=<5.9336 -0.0448 0.0401>, rot=<-0.0037 0.0213 0.298 -0.9543>}
 thumb_03_l {parent=thumb_02_l, loc=<2.2223 0.3682 -0.0907>, rot=<0.0118 -0.013 0.0066 -0.9998>}
 index_metacarpal_l {parent=hand_l, loc=<1.6535 0.0183 -2.96>, rot=<0.0158 -0.0124 -0.0173 -0.9996>}
 index_01_l {parent=index_metacarpal_l, loc=<8.3008 0.0357 -0.0524>, rot=<-0.0044 0.0485 0.0924 -0.9945>}
 index_02_l {parent=index_01_l, loc=<3.3587 -0.3004 0.0452>, rot=<-0.002 0.0027 0.3225 -0.9466>}
 index_03_l {parent=index_02_l, loc=<2.5578 -0.123 -0.0252>, rot=<-0.002 0.0067 0.2558 -0.9667>}
 middle_metacarpal_l {parent=hand_l, loc=<2.1856 -0.1725 -0.5984>, rot=<0.006 0.0168 -0.0329 -0.9993>}
 middle_01_l {parent=middle_metacarpal_l, loc=<7.3275 -0.2447 0.0129>, rot=<-0.0208 0.0042 0.0858 -0.9961>}
 middle_02_l {parent=middle_01_l, loc=<4.0123 0.1401 0.1175>, rot=<-0.0218 0.0258 0.3512 -0.9357>}
 middle_03_l {parent=middle_02_l, loc=<2.6077 0.0534 0.1373>, rot=<0.0082 0.0188 0.1708 -0.9851>}
 ring_metacarpal_l {parent=hand_l, loc=<2.2074 -0.3625 1.3638>, rot=<-0.0284 0.0722 -0.0458 -0.9959>}
 ring_01_l {parent=ring_metacarpal_l, loc=<7.0295 -0.6197 -0.1446>, rot=<-0.0489 -0.0252 0.1892 -0.9804>}
 ring_02_l {parent=ring_01_l, loc=<3.6174 0.4647 -0.1027>, rot=<0 0 0.2806 -0.9598>}
 ring_03_l {parent=ring_02_l, loc=<2.4152 0.114 0.0199>, rot=<0 0 0.1189 -0.9929>}
 pinky_metacarpal_l {parent=hand_l, loc=<1.7616 -0.9753 3.3814>, rot=<-0.0397 0.0814 -0.069 -0.9935>}
 pinky_01_l {parent=pinky_metacarpal_l, loc=<7.1709 -0.4831 -0.1094>, rot=<-0.0654 -0.0035 0.2242 -0.9723>}
 pinky_02_l {parent=pinky_01_l, loc=<2.8274 0.0893 -0.1043>, rot=<0.0037 0.0011 0.2854 -0.9584>}
 pinky_03_l {parent=pinky_02_l, loc=<2.2046 0.0307 -0.0078>, rot=<-0.0002 0.0098 0.1445 -0.9895>}
 lowerarm_twist_02_l {parent=lowerarm_l, loc=<11.5353 0 0>, rot=<0 0 0 1>}
 lowerarm_twist_01_l {parent=lowerarm_l, loc=<16.7573 0 0>, rot=<0 0 0 1>}
 upperarm_twist_01_l {parent=upperarm_l, loc=<7.2135 -0.1134 0>, rot=<0 0 0 1>}
 upperarm_twist_02_l {parent=upperarm_l, loc=<19.7998 -0.1134 0>, rot=<0 0 0 1>}
 clavicle_r {parent=spine_05, loc=<4.1651 -0.3023 5.0942>, rot=<0.7289 0.0465 -0.6829 0.009>}
 upperarm_r {parent=clavicle_r, loc=<-11.2117 1.1486 0.0332>, rot=<0.0179 -0.2982 -0.028 -0.9539>}
 lowerarm_r {parent=upperarm_r, loc=<-26.5013 0.0835 -0.2604>, rot=<0.0036 0.0161 -0.2298 -0.9731>}
 hand_r {parent=lowerarm_r, loc=<-23.1356 1.1606 0.2612>, rot=<-0.6704 -0.0028 0.0567 -0.7398>}
 thumb_01_r {parent=hand_r, loc=<-1.3118 0.3191 2.459>, rot=<0.6546 -0.4042 0.0664 -0.6354>}
 thumb_02_r {parent=thumb_01_r, loc=<-5.9336 0.0448 -0.0401>, rot=<-0.0037 0.0213 0.298 -0.9543>}
 thumb_03_r {parent=thumb_02_r, loc=<-2.2223 -0.3682 0.0907>, rot=<0.0118 -0.013 0.0066 -0.9998>}
 index_metacarpal_r {parent=hand_r, loc=<-1.6535 -0.0183 2.96>, rot=<0.0158 -0.0124 -0.0173 -0.9996>}
 index_01_r {parent=index_metacarpal_r, loc=<-8.3008 -0.0357 0.0524>, rot=<-0.0044 0.0485 0.0924 -0.9945>}
 index_02_r {parent=index_01_r, loc=<-3.3587 0.3004 -0.0452>, rot=<-0.002 0.0027 0.3225 -0.9466>}
 index_03_r {parent=index_02_r, loc=<-2.5578 0.123 0.0252>, rot=<-0.002 0.0067 0.2558 -0.9667>}
 middle_metacarpal_r {parent=hand_r, loc=<-2.1856 0.1725 0.5984>, rot=<0.006 0.0168 -0.0329 -0.9993>}
 middle_01_r {parent=middle_metacarpal_r, loc=<-7.3275 0.2447 -0.0129>, rot=<-0.0208 0.0042 0.0858 -0.9961>}
 middle_02_r {parent=middle_01_r, loc=<-4.0123 -0.1401 -0.1175>, rot=<-0.0218 0.0258 0.3512 -0.9357>}
 middle_03_r {parent=middle_02_r, loc=<-2.6077 -0.0534 -0.1373>, rot=<0.0082 0.0188 0.1708 -0.9851>}
 ring_metacarpal_r {parent=hand_r, loc=<-2.2074 0.3625 -1.3638>, rot=<-0.0284 0.0722 -0.0458 -0.9959>}
 ring_01_r {parent=ring_metacarpal_r, loc=<-7.0295 0.6197 0.1446>, rot=<-0.0489 -0.0252 0.1892 -0.9804>}
 ring_02_r {parent=ring_01_r, loc=<-3.6174 -0.4647 0.1027>, rot=<0 0 0.2806 -0.9598>}
 ring_03_r {parent=ring_02_r, loc=<-2.4152 -0.114 -0.0199>, rot=<0 0 0.1189 -0.9929>}
 pinky_metacarpal_r {parent=hand_r, loc=<-1.7616 0.9753 -3.3814>, rot=<-0.0397 0.0814 -0.069 -0.9935>}
 pinky_01_r {parent=pinky_metacarpal_r, loc=<-7.1709 0.4831 0.1094>, rot=<-0.0654 -0.0035 0.2242 -0.9723>}
 pinky_02_r {parent=pinky_01_r, loc=<-2.8274 -0.0893 0.1043>, rot=<0.0037 0.0011 0.2854 -0.9584>}
 pinky_03_r {parent=pinky_02_r, loc=<-2.2046 -0.0307 0.0078>, rot=<-0.0002 0.0098 0.1445 -0.9895>}
 weapon_r {parent=hand_r, loc=<0.1974 3.4111 -0.3811>, rot=<0 0 0 1>}
 hand_attach_r {parent=hand_r, loc=<-8.5001 2.9999 0>, rot=<0 0 -0.7071 -0.7071>}
 lowerarm_twist_02_r {parent=lowerarm_r, loc=<-11.5353 0 0>, rot=<0 0 0 -1>}
 lowerarm_twist_01_r {parent=lowerarm_r, loc=<-16.7573 0 0>, rot=<0 0 0 -1>}
 upperarm_twist_01_r {parent=upperarm_r, loc=<-7.2135 0.1134 0>, rot=<0 0 0 -1>}
 upperarm_twist_02_r {parent=upperarm_r, loc=<-19.7998 0.1134 0>, rot=<0 0 0 -1>}
 attach_cape {parent=spine_05, loc=<9.4656 -10.2402 0>, rot=<0.0616 -0.7044 -0.0616 -0.7044>}
 dyn_necklace {parent=spine_05, loc=<4.6702 7.9584 -0.0294>, rot=<-0.2136 -0.6741 -0.6741 -0.2136>}
 attach_backpack {parent=spine_05, loc=<-0.0002 -10.1459 0>, rot=<-0.0093 -0.707 0.0093 -0.707>}
 dyn_button_01 {parent=spine_05, loc=<4.2493 6.326 -10.9882>, rot=<0.0284 -0.7065 -0.0284 -0.7065>}
 dyn_button_02 {parent=spine_05, loc=<-2.4164 10.4473 -12.2818>, rot=<0.0284 -0.7065 -0.0284 -0.7065>}
 dyn_button_03 {parent=spine_05, loc=<-6.3939 9.5481 -13.471>, rot=<0.0284 -0.7065 -0.0284 -0.7065>}
 dyn_collar_l {parent=spine_05, loc=<9.2527 2.2966 -6.7203>, rot=<0.1582 -0.7351 -0.2622 -0.6048>}
 dyn_collar_c {parent=spine_05, loc=<11.1957 -5.5568 0>, rot=<0.0284 -0.7065 -0.0284 -0.7065>}
 dyn_collar_r {parent=spine_05, loc=<9.2527 2.2966 6.7203>, rot=<0.7351 0.1582 -0.6048 0.2622>}
 dyn_jacket_front_l {parent=spine_04, loc=<1.3123 9.6189 -13.5104>, rot=<-0.0159 0.3824 0.9231 -0.0384>}
 dyn_jacket_front_end_l {parent=dyn_jacket_front_l, loc=<5.5693 0 0>, rot=<0 0 0 1>}
 dyn_jacket_side_l {parent=spine_04, loc=<2.3553 -3.1297 -15.888>, rot=<0.0294 -0.7065 -0.7065 0.0294>}
 dyn_jacket_side_end_l {parent=dyn_jacket_side_l, loc=<6.9547 0 0>, rot=<0 0 0 1>}
 dyn_jacket_back_l {parent=spine_04, loc=<2.7847 -12.8148 0.0128>, rot=<0.0416 -0.9991 0 0>}
 dyn_jacket_back_end_l {parent=dyn_jacket_back_l, loc=<6.1736 0 0>, rot=<0 0 0 1>}
 dyn_jacket_front_r {parent=spine_04, loc=<1.3126 9.6189 13.5104>, rot=<-0.3824 -0.0159 -0.0384 -0.9231>}
 dyn_jacket_front_end_r {parent=dyn_jacket_front_r, loc=<-5.57 0 0>, rot=<0 0 0 1>}
 dyn_jacket_side_r {parent=spine_04, loc=<2.3558 -3.1296 15.888>, rot=<-0.7065 -0.0294 -0.0294 -0.7065>}
 dyn_jacket_side_end_r {parent=dyn_jacket_side_r, loc=<-6.955 0 0>, rot=<0 0 0 1>}
 thigh_l {parent=pelvis, loc=<-2.5624 -0.0642 -10.3834>, rot=<0.0004 0.0055 0.0155 -0.9999>}
 calf_l {parent=thigh_l, loc=<-39.2091 -0.1238 -0.6795>, rot=<0.0006 -0.0161 -0.0391 -0.9991>}
 foot_l {parent=calf_l, loc=<-43.624 -2.818 -0.713>, rot=<0.044 0.0061 0.0367 -0.9983>}
 ball_l {parent=foot_l, loc=<-4.6464 10.3094 -0.3137>, rot=<0.0001 -0.0001 -0.7063 -0.7079>}
 calf_twist_02_l {parent=calf_l, loc=<-17.5927 -1.7943 -0.2638>, rot=<0.0237 0.0095 -0.0035 -0.9997>}
 calf_twist_01_l {parent=calf_l, loc=<-34.2864 -1.9197 -0.5794>, rot=<0.0237 0.0095 -0.0035 -0.9997>}
 thigh_twist_01_l {parent=thigh_l, loc=<-23.0764 0 0>, rot=<0.0013 0 0 -1>}
 thigh_r {parent=pelvis, loc=<-2.5624 -0.0642 10.3834>, rot=<0.0055 -0.0004 0.9999 0.0155>}
 calf_r {parent=thigh_r, loc=<39.2091 0.1238 0.6795>, rot=<0.0006 -0.0161 -0.0391 -0.9991>}
 foot_r {parent=calf_r, loc=<43.624 2.818 0.713>, rot=<0.044 0.0061 0.0367 -0.9983>}
 ball_r {parent=foot_r, loc=<4.6464 -10.3094 0.3137>, rot=<0.0001 -0.0001 -0.7063 -0.7079>}
 calf_twist_02_r {parent=calf_r, loc=<17.5927 1.7943 0.2638>, rot=<0.0237 0.0095 -0.0035 -0.9997>}
 calf_twist_01_r {parent=calf_r, loc=<34.2864 1.9197 0.5794>, rot=<0.0237 0.0095 -0.0035 -0.9997>}
 thigh_twist_01_r {parent=thigh_r, loc=<23.0764 0 0>, rot=<0.0013 0 0 -1>}
 attach {parent=root, loc=<0 0 0>, rot=<0 0 0 1>}
 ik_hand_root {parent=root, loc=<0 0 0>, rot=<0 0 0 1>}
 ik_hand_gun {parent=ik_hand_root, loc=<-53.7993 3.4863 105.1719>, rot=<0.6513 0.0939 0.3837 -0.6479>}
 ik_hand_r {parent=ik_hand_gun, loc=<0 0 0>, rot=<0 0 0 1>}
 ik_hand_l {parent=ik_hand_gun, loc=<74.0247 66.6541 40.682>, rot=<0.688 0.6195 0.3781 0>}
 ik_foot_root {parent=root, loc=<0 0 0>, rot=<0 0 0 1>}
 ik_foot_l {parent=ik_foot_root, loc=<11.2777 -7.5397 9.7673>, rot=<0.0357 0.7045 0.0284 -0.7082>}
 ik_foot_r {parent=ik_foot_root, loc=<-11.2777 -7.5397 9.7673>, rot=<0.7082 0.0284 -0.7045 0.0357>}
 root_fp {parent=root, loc=<0 0 0>, rot=<0 0 0 1>}
 attach_fp {parent=root_fp, loc=<0 0 0>, rot=<0 0 0 1>}
 camera_root {parent=root_fp, loc=<0 0 0>, rot=<0 0 0 1>}
 camera {parent=camera_root, loc=<0 0 0>, rot=<0 0 0 1>}
 attach_emote_01 {parent=root, loc=<0 0 0>, rot=<0 0 0 1>}
 attach_emote_02 {parent=root, loc=<0 0 0>, rot=<0 0 0 1>}
 attach_emote_03 {parent=root, loc=<0 0 0>, rot=<0 0 0 1>}
 warp_point_01 {parent=root, loc=<0 0 0>, rot=<0 0 0 1>}
 warp_point_02 {parent=root, loc=<0 0 0>, rot=<0 0 0 1>}
 warp_point_03 {parent=root, loc=<0 0 0>, rot=<0 0 0 1>}
 dyn_ponytail_01_null {parent=root, type=Null, loc=<0.0598 -9.8681 166.3452>, rot=<-0.6647 0.6647 -0.2411 -0.2411>}
 dyn_ponytail_01_ctrl {controlType=EULER_TRANSFORM, parent=dyn_ponytail_01_null, type=Control, loc=<0 0 0>, rot=<0 0 0 1>}
 dyn_ponytail_02_null {parent=root, type=Null, loc=<-0.077 -14.7727 169.4408>, rot=<0.7047 -0.7047 -0.0588 -0.0588>}
 dyn_ponytail_02_ctrl {controlType=EULER_TRANSFORM, parent=dyn_ponytail_02_null, type=Control, loc=<124.3009 -14.9095 114.1657>, rot=<0.6156 -0.6647 -0.348 0.2411>}
 dyn_ponytail_03_null {parent=root, type=Null, loc=<-0.077 -22.8708 167.7836>, rot=<-0.6281 0.6281 0.3249 0.3249>}
 dyn_ponytail_03_ctrl {
  controlType=EULER_TRANSFORM
  parent=dyn_ponytail_03_null
  type=Control
  loc=<112.8715 -22.8708 124.0289>
  rot=<0.4552 -0.7047 -0.5411 -0.0588>
 }
 dyn_ponytail_04_null {parent=root, type=Null, loc=<-0.077 -27.5553 160.8748>, rot=<0.5455 -0.5455 -0.4499 -0.4499>}
 dyn_ponytail_04_ctrl {
  controlType=EULER_TRANSFORM
  parent=dyn_ponytail_04_null
  type=Control
  loc=<58.5152 -27.5553 149.9105>
  rot=<0.4385 -0.6281 -0.5547 -0.3249>
 }
 dyn_ponytail_05_null {parent=root, type=Null, loc=<-0.077 -29.0759 150.2598>, rot=<-0.5002 0.5002 0.4998 0.4998>}
 dyn_ponytail_05_ctrl {
  controlType=EULER_TRANSFORM
  parent=dyn_ponytail_05_null
  type=Control
  loc=<17.7943 -29.0759 149.0606>
  rot=<0.4504 -0.5455 -0.5451 -0.4499>
 }
 dyn_ponytail_06_null {parent=root, type=Null, loc=<-0.077 -29.0854 139.4989>, rot=<0.5002 -0.5002 -0.4998 -0.4998>}
 dyn_ponytail_06_ctrl {
  controlType=EULER_TRANSFORM
  parent=dyn_ponytail_06_null
  type=Control
  loc=<-10.8379 -29.0854 139.4989>
  rot=<0.5002 -0.4998 -0.5002 -0.4998>
 }
 dyn_ponytail_07_null {parent=root, type=Null, loc=<-0.077 -29.095 128.7381>, rot=<-0.5002 0.5002 0.4998 0.4998>}
 dyn_ponytail_07_ctrl {
  controlType=EULER_TRANSFORM
  parent=dyn_ponytail_07_null
  type=Control
  loc=<-10.8379 -29.095 128.7381>
  rot=<0.5002 -0.4998 -0.5002 -0.4998>
 }
 dyn_ponytail_08_null {parent=root, type=Null, loc=<-0.077 -29.1045 117.9772>, rot=<0.5002 -0.5002 -0.4998 -0.4998>}
 dyn_ponytail_08_ctrl {
  controlType=EULER_TRANSFORM
  parent=dyn_ponytail_08_null
  type=Control
  loc=<-10.8379 -29.1045 117.9772>
  rot=<0.5002 -0.4998 -0.5002 -0.4998>
 }
]
variables [
 dyn_Hair_BackL_ctrls {
  type="TArray<FRigElementKey>"
  default=((Type=Control,Name="dyn_Hair_BackL_02"),(Type=Control,Name="dyn_Hair_BackL_03"),(Type=Control,Name="dyn_Hair_BackL_04"),(Type=Control,Name="dyn_Hair_BackL_05"),(Type=Control,Name="dyn_Hair_BackL_06"))
 }
 dyn_Hair_BackL_pose {type=FRigPose, default=(Elements=,HierarchyTopologyVersion=-1,PoseHash=-1)}
 dyn_Hair_BackR_ctrls {
  type="TArray<FRigElementKey>"
  default=((Type=Control,Name="dfrm_arm_spline_01_r_ctrl"),(Type=Control,Name="dfrm_arm_spline_03_r_ctrl"),(Type=Control,Name="dfrm_arm_spline_05_r_ctrl"),(Type=Control,Name="dfrm_arm_spline_07_r_ctrl"),(Type=Control,Name="dfrm_arm_spline_09_r_ctrl"))
 }
 dyn_Hair_BackR_pose {type=FRigPose, default=(Elements=,HierarchyTopologyVersion=-1,PoseHash=-1)}
 dyn_Hair_Back_ctrls {
  type="TArray<FRigElementKey>"
  default=((Type=Control,Name="dyn_Hair_BackL_02"),(Type=Control,Name="dyn_Hair_BackL_03"),(Type=Control,Name="dyn_Hair_BackL_04"),(Type=Control,Name="dyn_Hair_BackL_05"),(Type=Control,Name="dyn_Hair_BackL_06"))
 }
 dyn_Hair_Back_pose {type=FRigPose, default=(Elements=,HierarchyTopologyVersion=-1,PoseHash=-1)}
]
