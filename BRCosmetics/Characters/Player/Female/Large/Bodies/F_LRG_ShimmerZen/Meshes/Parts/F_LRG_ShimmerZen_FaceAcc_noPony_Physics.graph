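# Physics asset for the no-ponytail face accessory mesh of the large female ShimmerZen cosmetic body.
# Capsules cover the upper spine, neck, head, clavicles, upper arms and collar bases, with a second, wider capsule on the left upper arm only.
# Two four-bone hair chains hang from the head on free joints, with twist locked and a tighter first swing limit between links.
# All body and collar joints rotate freely.
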PhysicsAsset /BRCosmetics/Characters/Player/Female/Large/Bodies/F_LRG_ShimmerZen/Meshes/Parts/F_LRG_ShimmerZen_FaceAcc_noPony_Physics
bodies {
 clavicle_l [{shapeType=Capsule, length=5.5517, radius=7.3175, shapeName=clavicle_l_capsule, center=<6.9223 0 0.217>, rotation=<rot -0.0001 -90 -91.7954>}]
 clavicle_r [{
   shapeType=Capsule
   length=5.6878
   radius=7.3175
   shapeName=clavicle_r_capsule
   center=<-6.0019 0 -0.1063>
   rotation=<rot -0.0001 -90 88.9855>
  }]
 dyn_collar_l_BASE [{
   shapeType=Capsule
   length=9.505
   radius=5.505
   shapeName=dyn_collar_l_BASE_capsule
   center=<3.8232 0.1929 2.2045>
   rotation=<rot -80.0002 -90 -0.0002>
  }]
 dyn_collar_r_BASE [{
   shapeType=Capsule
   length=9.505
   radius=4.505
   shapeName=dyn_collar_l_BASE_capsule
   center=<-3.8232 -0.1929 -2.2045>
   rotation=<rot -80.0002 -89.9999 -175.0002>
  }]
 dyn_hair_l_01 [{shapeType=Capsule, length=6, radius=1, shapeName=dyn_hair_l_01_capsule, rotation=<rot -59.9998 0 0>}]
 dyn_hair_l_02 [{shapeType=Capsule, length=6, radius=1, shapeName=dyn_hair_l_02_capsule, rotation=<rot -59.9998 0 0>}]
 dyn_hair_l_03 [{shapeType=Capsule, length=6, radius=1, shapeName=dyn_hair_l_03_capsule, rotation=<rot -59.9998 0 0>}]
 dyn_hair_l_04 [{shapeType=Capsule, length=6, radius=1, shapeName=dyn_hair_l_04_capsule, center=<0.2402 0.7618 0.1387>, rotation=<rot -59.9997 0 15>}]
 dyn_hair_r_01 [{shapeType=Capsule, length=6, radius=1, shapeName=dyn_hair_l_01_capsule, rotation=<rot -59.9998 0 180>}]
 dyn_hair_r_02 [{shapeType=Capsule, length=6, radius=1, shapeName=dyn_hair_l_02_capsule, rotation=<rot -59.9998 0 180>}]
 dyn_hair_r_03 [{shapeType=Capsule, length=6, radius=1, shapeName=dyn_hair_l_03_capsule, rotation=<rot -74.5704 -9.7254 -160.953>}]
 dyn_hair_r_04 [{
   shapeType=Capsule
   length=6
   radius=0.5
   shapeName=dyn_hair_l_04_capsule
   center=<-0.2402 -0.7618 -0.1387>
   rotation=<rot -72.5155 -21.3499 -135.3055>
  }]
 head [{
   shapeType=Capsule
   length=15.0416
   radius=6.4176
   shapeName=head_capsule
   center=<1.4489 3.1333 0.0066>
   rotation=<rot -0.2989 39.9995 89.7492>
  }]
 neck_01 [{shapeType=Capsule, length=0.505, radius=0.505, shapeName=neck_01_capsule, rotation=<rot 0 -90 0>}]
 spine_05 [{shapeType=Capsule, length=16.392, radius=9.2919, shapeName=spine_05_capsule, center=<13.513 -5.3842 0>, rotation=<rot 0 -90 0>}]
 upperarm_l [
  {
   shapeType=Capsule
   length=33.4871
   radius=5.2452
   shapeName=upperarm_l_capsule
   center=<15.142 0.0017 0.2617>
   rotation=<rot -0.0074 -89.9934 -90.99>
  }
  {
   shapeType=Capsule
   length=9.2852
   radius=12.4219
   shapeName=upperarm_l_capsule
   center=<7.5399 1.5268 4.6483>
   rotation=<rot -84.2843 0.0009 -0.0008>
  }
 ]
 upperarm_r [{
   shapeType=Capsule
   length=35.0128
   radius=5.2653
   shapeName=upperarm_r_capsule
   center=<-15.7143 0 0.3059>
   rotation=<rot -0.0001 -89.9999 91.1152>
  }]
}
constraints [
 {
  bone1Name=head
  bone2Name=neck_01
  swing1LimitDegrees=45
  swing1Motion=Free
  swing2LimitDegrees=45
  swing2Motion=Free
  twistLimitDegrees=45
  twistMotion=Free
 }
 {
  bone1Name=neck_01
  bone2Name=spine_05
  swing1LimitDegrees=45
  swing1Motion=Free
  swing2LimitDegrees=45
  swing2Motion=Free
  twistLimitDegrees=45
  twistMotion=Free
 }
 {
  bone1Name=clavicle_l
  bone2Name=spine_05
  swing1LimitDegrees=45
  swing1Motion=Free
  swing2LimitDegrees=45
  swing2Motion=Free
  twistLimitDegrees=45
  twistMotion=Free
 }
 {
  bone1Name=clavicle_r
  bone2Name=spine_05
  swing1LimitDegrees=45
  swing1Motion=Free
  swing2LimitDegrees=45
  swing2Motion=Free
  twistLimitDegrees=45
  twistMotion=Free
 }
 {
  bone1Name=upperarm_l
  bone2Name=clavicle_l
  swing1LimitDegrees=45
  swing1Motion=Free
  swing2LimitDegrees=45
  swing2Motion=Free
  twistLimitDegrees=45
  twistMotion=Free
 }
 {
  bone1Name=upperarm_r
  bone2Name=clavicle_r
  swing1LimitDegrees=45
  swing1Motion=Free
  swing2LimitDegrees=45
  swing2Motion=Free
  twistLimitDegrees=45
  twistMotion=Free
 }
 {
  bone1Name=dyn_collar_l_BASE
  bone2Name=spine_05
  swing1LimitDegrees=45
  swing1Motion=Free
  swing2LimitDegrees=45
  swing2Motion=Free
  twistLimitDegrees=45
  twistMotion=Free
 }
 {
  bone1Name=dyn_collar_r_BASE
  bone2Name=spine_05
  swing1LimitDegrees=45
  swing1Motion=Free
  swing2LimitDegrees=45
  swing2Motion=Free
  twistLimitDegrees=45
  twistMotion=Free
 }
 {
  bone1Name=dyn_hair_l_01
  bone2Name=head
  swing1LimitDegrees=45
  swing1Motion=Free
  swing2LimitDegrees=45
  swing2Motion=Free
  twistLimitDegrees=45
  twistMotion=Free
 }
 {bone1Name=dyn_hair_l_02, bone2Name=dyn_hair_l_01, swing1LimitDegrees=10, swing2LimitDegrees=45, twistLimitDegrees=45, twistMotion=Locked}
 {bone1Name=dyn_hair_l_03, bone2Name=dyn_hair_l_02, swing1LimitDegrees=10, swing2LimitDegrees=45, twistLimitDegrees=45, twistMotion=Locked}
 {bone1Name=dyn_hair_l_04, bone2Name=dyn_hair_l_03, swing1LimitDegrees=10, swing2LimitDegrees=45, twistLimitDegrees=45, twistMotion=Locked}
 {
  bone1Name=dyn_hair_r_01
  bone2Name=head
  swing1LimitDegrees=45
  swing1Motion=Free
  swing2LimitDegrees=45
  swing2Motion=Free
  twistLimitDegrees=45
  twistMotion=Free
 }
 {bone1Name=dyn_hair_r_02, bone2Name=dyn_hair_r_01, swing1LimitDegrees=10, swing2LimitDegrees=45, twistLimitDegrees=45, twistMotion=Locked}
 {bone1Name=dyn_hair_r_03, bone2Name=dyn_hair_r_02, swing1LimitDegrees=10, swing2LimitDegrees=45, twistLimitDegrees=45, twistMotion=Locked}
 {bone1Name=dyn_hair_r_04, bone2Name=dyn_hair_r_03, swing1LimitDegrees=10, swing2LimitDegrees=45, twistLimitDegrees=45, twistMotion=Locked}
]
