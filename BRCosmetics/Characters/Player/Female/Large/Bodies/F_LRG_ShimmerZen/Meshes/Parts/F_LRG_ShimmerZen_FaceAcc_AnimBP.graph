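# Animation Blueprint for the ShimmerZen face accessory part, driving ponytail motion on the character.
# Both branches copy the character mesh pose: one runs full RigidBody physics from the head, the other a Control Rig, no-pony physics from spine_05 and bone offsets.
# A backbling check blends to the second branch for backblings that are neither long nor capes.
# Movement-state gravity and wind parameters feed both simulations; cape alpha is set but unused in the AnimGraph.
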AnimBlueprint /BRCosmetics/Characters/Player/Female/Large/Bodies/F_LRG_ShimmerZen/Meshes/Parts/F_LRG_ShimmerZen_FaceAcc_AnimBP
target_skeleton=&/F_LRG_ShimmerZen_FaceAcc_Skeleton
graphs [
 AnimGraph {nodes [
   n1 = "Transform (Modify) Bone - Bone: dyn_ponytail_01"(ComponentPose=Local To Component(LocalPose=Use cached pose 'HairDynPose'()), Alpha=1, Translation="0,0,0", Rotation="0,0,14", Scale="1,1,1") {
    rotationMode=BMM_Additive
    translationMode=BMM_Additive
    alphaBoolBlend {blendOption=Linear}
    boneToModify {boneName=dyn_ponytail_01}
    rotation {roll=14}
   }
   n2 = "Transform (Modify) Bone - Bone: dyn_ponytail_04"!(
    ComponentPose=Local To Component(LocalPose=Use cached pose 'HairCachePose'())
    bAlphaBoolEnabled:bool=True
    Translation="0,0,0"
    Rotation="10,0,0"
    Scale="1,0.65,0.65"
   ) -> (Pose) {
    alphaInputType=Bool
    rotationMode=BMM_Additive
    rotationSpace=BCS_BoneSpace
    scaleMode=BMM_Additive
    scaleSpace=BCS_BoneSpace
    translationMode=BMM_Additive
    translationSpace=BCS_BoneSpace
    alphaBoolBlend {blendOption=Linear}
    boneToModify {boneName=dyn_ponytail_04}
    rotation {pitch=10}
    scale {y=0.65, z=0.65}
   }
   n3 = "Transform (Modify) Bone - Bone: dyn_ponytail_07"!(ComponentPose=@n2, bAlphaBoolEnabled:bool=True, Translation="0,0,-10", Rotation="-20,0,0", Scale="1,1,1") -> (Pose) {
    alphaInputType=Bool
    rotationMode=BMM_Additive
    rotationSpace=BCS_BoneSpace
    scaleMode=BMM_Additive
    scaleSpace=BCS_BoneSpace
    translationMode=BMM_Additive
    translationSpace=BCS_BoneSpace
    alphaBoolBlend {blendOption=Linear}
    boneToModify {boneName=dyn_ponytail_07}
    rotation {pitch=-20}
    translation {z=-10}
   }
   Output Pose(Result=Two Way Blend(A=Component To Local(ComponentPose=@n1), B=Component To Local(ComponentPose=@n3), Alpha=Get backbling alpha() -> (backbling alpha:double)) {blendNode {alphaBoolBlend {blendOption=Linear}}}) {layerGroup=None}
   n4 = Control Rig(Source=Copy Pose From Mesh(SourceMeshComponent=Get Character Mesh() -> (Character Mesh:Skeletal Mesh Component Object Reference)) {bUseAttachedParent=true}) {
    alphaBoolBlend {blendOption=Linear}
    controlRigAssetReference {blueprintRigClass=&/F_LRG_ShimmerZen_FaceAcc_CtrlRig.F_LRG_ShimmerZen_FaceAcc_CtrlRig_C}
   }
   n5 = RigidBody!(
    ComponentPose=Local To Component(LocalPose=@n4)
    Alpha
    OverrideWorldGravity:Vector=Get Gravity Override Cape() -> (Gravity Override Cape:Vector)
    ExternalForce
    ComponentLinearAccScale:Vector="0.4,0.25,0.2"
    ComponentLinearVelScale:Vector="0.45,0.2,0.15"
   ) -> (Pose) {
    bClampLinearTranslationLimitToRefPose=true
    bOverrideWorldGravity=true
    overridePhysicsAsset=&/F_LRG_ShimmerZen_FaceAcc_noPony_Physics
    simulationSpace=BaseBoneSpace
    alphaBoolBlend {blendOption=Linear}
    baseBoneRef {boneName=spine_05}
    componentAppliedLinearAccClamp=<3000 1500 1000>
    componentLinearAccScale=<0.1 0.1 0.1>
    componentLinearVelScale=<0.1 0.1 0.1>
    overrideWorldGravity {z=-980}
    simSpaceSettings {
     maxAngularAcceleration=1000
     maxAngularVelocity=10
     maxLinearAcceleration=700
     maxLinearVelocity=500
     velocityScaleZ=0
     worldAlpha=0.1
     externalLinearDragV=<0.5 0.5>
    }
   }
   Save cached pose 'HairCachePose'(Pose=Component To Local(ComponentPose=@n5)) {cacheName=HairCachePose}
   n6 = Get Character Mesh()
   n7 = RigidBody(
    ComponentPose=Local To Component(LocalPose=Copy Pose From Mesh(SourceMeshComponent=@n6) {bUseAttachedParent=true})
    OverrideWorldGravity=Get Gravity Override Cape()
    ComponentLinearAccScale="0.4,0.4,0.4"
    ComponentLinearVelScale="0.4,0.4,0.4"
   ) {
    bClampLinearTranslationLimitToRefPose=true
    bOverrideWorldGravity=true
    overridePhysicsAsset=&/F_LRG_ShimmerZen_FaceAcc_Physics
    simulationSpace=BaseBoneSpace
    alphaBoolBlend {blendOption=Linear}
    baseBoneRef {boneName=head}
    componentAppliedLinearAccClamp=<3000 1500 1000>
    componentLinearAccScale=<0.1 0.1 0.1>
    componentLinearVelScale=<0.1 0.1 0.1>
    overrideWorldGravity {z=-980}
    simSpaceSettings {
     maxAngularAcceleration=1000
     maxAngularVelocity=10
     maxLinearAcceleration=700
     maxLinearVelocity=500
     velocityScaleZ=0
     worldAlpha=0.1
     externalLinearDragV=<0.5 0.5>
    }
   }
   Save cached pose 'HairDynPose'(Pose=Component To Local(ComponentPose=@n7)) {cacheName=HairDynPose}
  ]}
 EventGraph {nodes [
   Event BlueprintUpdateAnimation()
   > n1 = Sequence!(execute) -> (then_0, then_1, then_2:Exec)
   > n2 = Gravity Override State Function DEPRECATED(CharacterPartAnimInstance=Self-Reference(), GameTimeInSeconds=GetGameTimeInSeconds(), GravityOverrideParams=Get Gravity Override Params())
   > Set Gravity Override Cape(Gravity Override Cape:Vector=@n2) -> (Output_Get:Vector)
   n3 = CheckForBackbling()
   n4 = OR Boolean(A=@n3.bIsBackblingLong, B=@n3.bIsBackblingACape)
   n5 = Branch(execute=@n1.then_1, Condition=AND Boolean(A=@n3, B=NOT Boolean(A=@n4)))
   > Set backbling alpha(backbling alpha:double=1) -> (Output_Get:double)
   Set backbling alpha(execute=@n5.else, backbling alpha=0)
   n6 = Branch(execute=@n1.then_2, Condition=@n4)
   > Set cape alpha(cape alpha:double=1) -> (Output_Get:double)
   Set cape alpha(execute=@n6.else, cape alpha=0)
  ]}
]
variables [
 backbling alpha {type=number}
 Body Mesh {type=SkeletalMeshComponent}
 cape alpha {type=number}
 Character Mesh {type=SkeletalMeshComponent}
 Gravity Override Cape {type=Vector}
 Gravity Override Params {
  type=GravityOverrideParamsStruct
  default {
   onGround_Standing {windAmplitude=<-10 -150 0>, windFrequency=<-7 -1 -1>, windOffset=<-10 -100 0>}
   skydive_DivingDown {jointName=root, gravityOverride=<0 1000 0>, windOffset=<0 0 50>}
   skydive_FloatingDown {gravityOverride=<0 -400 600>, windAmplitude=<50 -100 100>, windOffset=<0 -150 100>}
   skydive_GoingUp {gravityOverride=<0 -300 600>}
   skydive_Parachuting {gravityOverride=<0 -400 -800>, windAmplitude=<50 -100 500>, windOffset=<0 -150 300>}
   skydive_Surfing {gravityOverride=<0 -300 300>}
  }
 }
 Head_Mesh {type=SkeletalMeshComponent}
 Is Dive or Dive Floating or Wearing or DBNO {type=boolean}
 is_WearingCape {type=boolean}
 isWearingPackBool {type=boolean}
 IsWearingPackBool or DBNO or DBNOCarrying {type=number}
 isWearingPackBool_Dliving {type=boolean}
 neck_rot_l {type=number}
 neck_rot_r {type=number}
 neck_rot_up {type=number}
 neckRotateBack {type=number}
 rban alpha {type=number}
]
